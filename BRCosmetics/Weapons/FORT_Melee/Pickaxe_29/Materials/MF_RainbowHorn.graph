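# Material function that overlays an animated, glowing rainbow on the incoming material's emissive and base color for the Pickaxe_29 horn.
# A tiled, rotated, panning rainbow gradient is wobbled by scrolling noise, then expanded to HDR for emissive glow.
# A mask texture and Fresnel term control the blend, with HornGlowINT scaling the glow and base color tint.
# Low quality swaps in an unwobbled gradient with a constant brightness multiplier.
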MaterialFunction /BRCosmetics/Weapons/FORT_Melee/Pickaxe_29/Materials/MF_RainbowHorn
graph {nodes [
  n1 = GetMaterialAttributes(None=FunctionInput() {bUsePreviewValueAsDefault=true, inputName=Material, inputType=FunctionInput_MaterialAttributes}) {attributeGetTypes [#1, #2]}
  n2 = ScalarParameter() {group=Rainbow Horn, parameterName=HornGlowINT}
  n3 = TextureSample(UVs=Panner() {speedX=-0.1, speedY=-0.005}) {samplerType=SAMPLERTYPE_LinearGrayscale, texture=/Engine/Functions/Engine_MaterialFunctions02/ExampleContent/Textures/LowResBlurredNoise}
  n4 = ComponentMask(None=@n3.RGB) {g=true, r=true}
  n5 = ScalarParameter() {defaultValue=1, group=Rainbow Horn, parameterName=Wobble}
  n6 = TextureCoordinate()
  n7 = ScalarParameter() {defaultValue=8, group=Rainbow Horn, parameterName=RainbowTIling}
  n8 = Panner(
   Coordinate=MaterialFunctionCall(UVs=Multiply(A=@n6, B=@n7), "Rotation Angle (0-1)"=ScalarParameter() {defaultValue=0.25, group=Rainbow Horn, parameterName=Rot}) {materialFunction=&CustomRotator}.Rotated Values
   Speed=AppendVector(
    A=ScalarParameter() {defaultValue=-0.4, group=Rainbow Horn, parameterName=Rainbow_PanX}
    B=ScalarParameter() {group=Rainbow Horn, parameterName=Rainbow_PanY}
   )
  )
  n9 = TextureObject() {texture=/Game/Global/Textures/Default/Masks_and_Gradients/T_Grad_Rainbow}
  n10 = TextureSample(UVs=Add(A=Multiply(A=@n4, B=@n5), B=@n8), Tex=@n9) {texture=/Game/Global/Textures/Default/Masks_and_Gradients/FortniteRainbow}
  n11 = MaterialFunctionCall(
   LDR Input=@n10.RGB
   HDR Dynamic Range=ScalarParameter() {defaultValue=1000, group=Rainbow Horn, parameterName=DynamicRange}
   HDR Tint=@n10.RGB
   HDR Intensity Multiplier=ScalarParameter() {defaultValue=5, group=Rainbow Horn, parameterName=RainbowHornBrightness}
  ) {materialFunction=&DeriveHDRfromLDR}
  n12 = Multiply(A=TextureSample(UVs=@n8, Tex=@n9) {texture=&FortniteRainbow}.RGB, B=Constant() {r=10})
  n13 = Add(
   A=@n1.EmissiveColor
   B=Multiply(A=@n2, B=QualitySwitch(Default=@n11.RemappedRange, Low=@n12, High=@n11.RemappedRange, Medium=@n11.RemappedRange))
  )
  n14 = TextureSampleParameter2D() {
   group=Weapon_Textures
   parameterName=RainbowMask
   samplerType=SAMPLERTYPE_Masks
   texture=/BRCosmetics/Weapons/FORT_Melee/Pickaxe_29/Textures/T_UnicornHorn_FX
  }
  n15 = ScalarParameter() {defaultValue=0.02, group=Rainbow Horn, parameterName=Rainbow_Min}
  n16 = ScalarParameter() {defaultValue=0.5, group=Rainbow Horn, parameterName=RainbowFres_Min}
  n17 = ScalarParameter() {defaultValue=15, group=Rainbow Horn, parameterName=RainbowFres_Max}
  n18 = Clamp(None=Fresnel(ExponentIn=ScalarParameter() {defaultValue=2.5, group=Rainbow Horn, parameterName=Fres_EX}) {baseReflectFraction=0})
  n19 = Multiply(A=@n14.R, B=Multiply(A=LinearInterpolate(A=@n15, Alpha=@n14.G) {constA=0.1}, B=LinearInterpolate(A=@n16, B=@n17, Alpha=@n18)))
  n20 = Multiply(A=@n19, B=@n2)
  FunctionOutput(None=SetMaterialAttributes!(
    MaterialAttributes=@n1.MaterialAttributes
    EmissiveColor=LinearInterpolate(A=@n1.EmissiveColor, B=@n13, Alpha=@n19)
    BaseColor=LinearInterpolate(A=@n1.BaseColor, B=@n10.RGB, Alpha=@n20)
   ) {attributeSetTypes [#1, #2]})
 ]}
