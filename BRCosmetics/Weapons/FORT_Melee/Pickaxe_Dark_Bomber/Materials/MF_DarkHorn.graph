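# Material function for the Dark Bomber pickaxe that adds an animated, smoky glow to the horn through the incoming material attributes' emissive color.
# A world-space sphere mask, UVs and panning noise distort a scrolling smoke trail, from which two multiplied noise layers are subtracted.
# The horn mask's green channel boosts the result, which is tinted by the emissive texture.
# DeriveHDRfromLDR remaps brightness except on Low quality, and the mask's red channel confines the glow.
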MaterialFunction /BRCosmetics/Weapons/FORT_Melee/Pickaxe_Dark_Bomber/Materials/MF_DarkHorn
graph {nodes [
  n1 = GetMaterialAttributes(None=FunctionInput() {bUsePreviewValueAsDefault=true, inputName=Material, inputType=FunctionInput_MaterialAttributes}) {attributeGetTypes [#1]}
  n2 = SphereMask(A=WorldPosition().XYZ, B=ObjectPositionWS(), Radius=Constant() {r=220}, Hardness=Constant())
  n3 = Multiply(A=@n2, B=@n2)
  n4 = ComponentMask(None=TextureCoordinate()) {g=true}
  n5 = Panner(Coordinate=TextureCoordinate() {uTiling=4, vTiling=1.2}) {speedY=-0.8}
  n6 = Add(
   A=AppendVector(A=Multiply(A=@n3) {constB=5}, B=Multiply(A=@n4))
   B=ComponentMask(None=TextureSample(UVs=@n5) {samplerType=SAMPLERTYPE_LinearColor, texture=&T_NoisePack_LinearColor}.RGB) {g=true, r=true}
  )
  n7 = ScalarParameter() {defaultValue=0.45, parameterName=Smoke Trail Distortion Intensity}
  n8 = TextureSampleParameter2D(UVs=Panner(Coordinate=Multiply(A=@n6, B=@n7) {constB=0.45}) {speedX=1.5}) {parameterName=SmokeTrailTexture, samplerType=SAMPLERTYPE_Grayscale, texture=&T_Smoke_Trail_Soft}
  n9 = Multiply(A=TextureCoordinate() {uTiling=4, vTiling=4}, B=AppendVector(A=Constant() {r=1}, B=Constant() {r=0.35}))
  n10 = Add(A=@n9) {constB=0.3}
  n11 = Multiply(
   A=TextureSample(UVs=Panner(Coordinate=@n10) {speedX=0.25, speedY=1.9}) {samplerType=SAMPLERTYPE_LinearColor, texture=&T_NoisePack_LinearColor}.R
   B=TextureSample(UVs=Panner(Coordinate=@n9) {speedX=-0.18, speedY=1.2}) {samplerType=SAMPLERTYPE_LinearColor, texture=&T_NoisePack_LinearColor}.G
  )
  n12 = Subtract(A=Multiply(A=@n8.R, B=@n8.R), B=Multiply(A=@n11, B=@n11))
  n13 = TextureSampleParameter2D() {
   group=Weapon_Textures
   parameterName=HornGlowMask
   texture=/BRCosmetics/Weapons/FORT_Melee/Pickaxe_Dark_Bomber/Textures/T_Pickaxe_Dark_Bomber_HornMask
  }
  n14 = Clamp(None=Multiply(A=@n13.G) {constB=10})
  n15 = TextureSampleParameter2D() {group=Emissive, parameterName=EmissiveTexture, texture=/BRCosmetics/Weapons/FORT_Melee/Pickaxe_Dark_Bomber/Textures/T_Pickaxe_Dark_Bomber_E}
  n16 = Multiply(A=Multiply(A=Add(A=@n12, B=@n14), B=@n15.RGB), B=LinearInterpolate(Alpha=@n14) {constA=1, constB=10})
  n17 = MaterialFunctionCall(
   LDR Input=Clamp(None=@n16)
   HDR Tint=@n16
   HDR Intensity Multiplier=ScalarParameter() {defaultValue=5, group=Rainbow Horn, parameterName=RainbowHornBrightness}
  ) {materialFunction=&DeriveHDRfromLDR}
  n18 = Multiply(A=@n16, B=Constant() {r=10})
  n19 = Add(
   A=@n1.EmissiveColor
   B=Multiply(
    A=ScalarParameter() {group=Rainbow Horn, parameterName=HornGlowINT}
    B=QualitySwitch(Default=@n17.RemappedRange, Low=@n18, High=@n17.RemappedRange, Medium=@n17.RemappedRange)
   )
  )
  n20 = TextureSampleParameter2D() {
   group=Weapon_Textures
   parameterName=HornGlowMask
   texture=&T_Pickaxe_Dark_Bomber_HornMask
  }
  FunctionOutput(None=SetMaterialAttributes!(MaterialAttributes=@n1.MaterialAttributes, EmissiveColor=LinearInterpolate(A=@n1.EmissiveColor, B=@n19, Alpha=@n20.R)) {attributeSetTypes [#1]})
 ]}
