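# Master material for the Pickaxe 29 melee cosmetic, with a static switch choosing between Brite Bomber and Dark Bomber horn looks.
# The gun master shader function feeds the elemental weapons function on the default shading path, while mobile skips that step.
# The result runs through the ADS scope master, then the rainbow or dark horn function, and finally the accessory tech-art overrides.
# It uses material attributes and supports skeletal meshes and mesh deformers.
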Material /BRCosmetics/Weapons/FORT_Melee/Pickaxe_29/Materials/M_Pickaxe29_MASTER
graph {
 properties {
  bUseMaterialAttributes=true
  bUsedWithMeshDeformer=true
  bUsedWithSkeletalMesh=true
  numCustomizedUVs=5
  pixelDepthOffsetMode=PDOM_Legacy
  usedShadingModels=Default Lit
 }
 nodes [
  n1 = MaterialFunctionCall!(UV 1 coordinates, MuzzleFalloff) {materialFunction=/Game/Packages/Fortnite_UniqueMaterials_1/Weapons/M_GunMasterShaderFunction}
  n2 = MaterialFunctionCall!(Material Attributes=ShadingPathSwitch(
    Default=MaterialFunctionCall!(In=@n1.Material Result, Pre-Skinned Local Position, Pre-Skinned Local Normal) {materialFunction=/Game/Weapons/WeaponSkins/MaterialFunction/ElementalWeaponsFunction}.Result
    Mobile=@n1.Material Result
   ) {desc=Migrated from FeatureLevelSwitch}) {materialFunction=/Game/Packages/Fortnite_UniqueMaterials_1/Weapons/ADSScopeMaster}
  MaterialFunctionCall(In=StaticSwitchParameter(
    True=MaterialFunctionCall!(Material=@n2.Result) {materialFunction=&/MF_RainbowHorn}.Result
    False=MaterialFunctionCall!(Material=@n2.Result) {materialFunction=/BRCosmetics/Weapons/FORT_Melee/Pickaxe_Dark_Bomber/Materials/MF_DarkHorn}.Result
   ) {defaultValue=true, parameterName=Brite Bomber or Dark Bomber ?}) {materialFunction=&MF_TechArtOverrides_accessories}
  Reroute(None=@n1.Normal Texture)
  StaticBoolParameter() {parameterName=Use Energy}
  StaticBoolParameter() {parameterName=Use Lightning}
  StaticBoolParameter() {parameterName=Use Fire}
  StaticBoolParameter() {parameterName=Use Ice}
 ]
}
