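# Animation Blueprint for the Pickaxe_29 melee cosmetic skeleton that scales its unicorn horn bone.
# The AnimGraph lerps between minimum and maximum scale vectors using HornScale, then applies the result to HornBone through an additive Transform (Modify) Bone node.
# The update event in the EventGraph is empty, and nothing in this Blueprint sets HornScale.
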AnimBlueprint /BRCosmetics/Weapons/FORT_Melee/Pickaxe_29/Meshes/ANimBP_UnicornHorn
target_skeleton=&/SK_Pickaxe_029_Skeleton
graphs [
 AnimGraph {nodes [
   n1 = "Lerp (Vector)"!(
    A:Vector=Get MinHornScale() -> (MinHornScale:Vector)
    B:Vector=Get MaxHornScale() -> (MaxHornScale:Vector)
    Alpha:float=Get HornScale() -> (HornScale:double)
   ) -> (ReturnValue:Vector)
   Output Pose(Result=Component To Local(ComponentPose="Transform (Modify) Bone - Bone: HornBone"(Alpha=1, Translation="0,0,0", Rotation="0,0,0", Scale=@n1) {scaleMode=BMM_Additive, alphaBoolBlend {blendOption=Linear}, boneToModify {boneName=HornBone}})) {layerGroup=None}
  ]}
 EventGraph {nodes [Event BlueprintUpdateAnimation()]}
]
variables [
 HornScale {type=number}
 MaxHornScale {type=Vector, default=<1.5 1.5 1.5>}
 MinHornScale {type=Vector, default=<1 1 1>}
]
