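# Physics asset for the Pickaxe_029 melee weapon skeletal mesh in the Battle Royale cosmetics set.
# A single box body on the BoneBone bone covers the weapon as one elongated, offset and rotated volume.
# It has no constraints, so it provides one simple rigid collision shape with no articulated parts.
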PhysicsAsset /BRCosmetics/Weapons/FORT_Melee/Pickaxe_29/Meshes/SK_Pickaxe_029_PhysicsAsset
bodies {BoneBone [{shapeType=Box, extentX=45.2619, extentY=18.2727, extentZ=141.8045, center=<0.7638 1.6101 -32.433>, rotation=<rot 0.0293 -90 0.102>}]}
constraints []
